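# Physics asset for the Venom vehicle body skeletal mesh, filed under Old, giving each major bone a single capsule body.
# Constraints form a tree from the frame to the body shell with glass, the front differential and both rear axle pivots.
# The differential feeds steering knuckles, wheels and shocks; rear pivots carry wheels and shocks, with a brake pad only on the left.
# All joints share identical swing and twist limit angles.
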PhysicsAsset /VehicleCosmetics/Bodies/Venom/Meshes/Old/SK_Body_Venom_PhysicsAsset
bodies {
 axle_pivot_bk_l [{shapeType=Capsule, length=26.5368, radius=10.9925, center=<0.1746 -68.8871 0.6975>, rotation=<rot 0 179.5619 -89.9685>}]
 axle_pivot_bk_r [{shapeType=Capsule, length=33.4847, radius=17.5526, center=<-2.0029 74.2431 3.1189>, rotation=<rot 0 -163.2055 -74.098>}]
 body [{shapeType=Capsule, length=227.5709, radius=125.6493, center=<7.6941 -0.001 20.012>, rotation=<rot 0 -89.9698 -82.4263>}]
 brakepad_bk_l [{shapeType=Capsule, length=17.9521, radius=14.6767, center=<-5.979 0.2889 6.7576>, rotation=<rot -0.4039 -90 -45.5677>}]
 frame [{shapeType=Capsule, length=0.505, radius=0.505, rotation=<rot 0 -90 0>}]
 front_differential [{shapeType=Capsule, length=51.8423, radius=9.0328, center=<-0.0007 0 -0.048>, rotation=<rot 0 0 -90>}]
 glass_c [{shapeType=Capsule, length=46.3882, radius=43.8907, center=<-20.4182 0 8.1535>, rotation=<rot 0 -90 -64.3765>}]
 glass_l [{shapeType=Capsule, length=64.7765, radius=12.3852, center=<-24.7158 4.9119 10.628>, rotation=<rot 0 -95.4694 -83.8732>}]
 glass_r [{shapeType=Capsule, length=64.7765, radius=12.3852, center=<24.7158 -4.9119 -10.628>, rotation=<rot 0 -95.4694 -83.8732>}]
 shock_base_bk_l [{shapeType=Capsule, length=27.562, radius=12.4927, center=<-0.2482 6.2246 17.717>, rotation=<rot 30.5069 -90 0.4358>}]
 shock_base_bk_r [{shapeType=Capsule, length=27.5702, radius=12.4314, center=<-0.2343 -6.1689 17.7453>, rotation=<rot -30.1745 -90 0.3837>}]
 shock_base_fr_l [{shapeType=Capsule, length=22.2441, radius=8.8091, center=<-1.0619 1.0064 13.7574>, rotation=<rot 8.7118 -90 -8.6537>}]
 shock_base_fr_r [{shapeType=Capsule, length=22.2442, radius=8.8091, center=<-1.0622 -1.0064 13.7576>, rotation=<rot -8.7113 -90 -8.6568>}]
 steering_knuckle_fr_l [{shapeType=Capsule, length=21.4803, radius=10.9926, center=<-0.0001 -26.0046 0.2306>, rotation=<rot 0 180 -89.3748>}]
 steering_knuckle_fr_r [{shapeType=Capsule, length=21.4803, radius=10.9926, center=<-0.0001 26.0046 0.2306>, rotation=<rot 0 0 -89.3748>}]
 wheel_disc_bk_l [{shapeType=Capsule, length=25.3798, radius=25.3798, center=<-0.0354 -1.5012 0.6193>, rotation=<rot 0 -90 -90.0003>}]
 wheel_disc_bk_r [{shapeType=Capsule, length=25.3798, radius=25.3798, center=<-0.0354 1.5081 0.6193>, rotation=<rot 0 -90 -90.0002>}]
 wheel_disc_fr_l [{shapeType=Capsule, length=23.8533, radius=23.8532, center=<-0.0007 -1.9673 0.0003>, rotation=<rot 0 -90 -90.0009>}]
 wheel_disc_fr_r [{shapeType=Capsule, length=23.8532, radius=23.8532, center=<-0.0007 1.9673 0.0003>, rotation=<rot 0 -90 -90.0001>}]
 wheel_steering_fr_l [{shapeType=Capsule, length=17.6414, radius=13.7891, center=<-5.2906 -4.038 6.6141>, rotation=<rot 0 -86.0765 -51.9415>}]
 wheel_steering_fr_r [{shapeType=Capsule, length=17.6414, radius=13.7891, center=<-5.2906 4.038 6.6141>, rotation=<rot 0 -93.9235 -51.9413>}]
}
constraints [
 {bone1Name=body, bone2Name=frame, swing1LimitDegrees=45, swing2LimitDegrees=45, twistLimitDegrees=45}
 {bone1Name=glass_c, bone2Name=body, swing1LimitDegrees=45, swing2LimitDegrees=45, twistLimitDegrees=45}
 {bone1Name=glass_l, bone2Name=body, swing1LimitDegrees=45, swing2LimitDegrees=45, twistLimitDegrees=45}
 {bone1Name=glass_r, bone2Name=body, swing1LimitDegrees=45, swing2LimitDegrees=45, twistLimitDegrees=45}
 {bone1Name=front_differential, bone2Name=frame, swing1LimitDegrees=45, swing2LimitDegrees=45, twistLimitDegrees=45}
 {bone1Name=steering_knuckle_fr_l, bone2Name=front_differential, swing1LimitDegrees=45, swing2LimitDegrees=45, twistLimitDegrees=45}
 {bone1Name=wheel_steering_fr_l, bone2Name=steering_knuckle_fr_l, swing1LimitDegrees=45, swing2LimitDegrees=45, twistLimitDegrees=45}
 {bone1Name=wheel_disc_fr_l, bone2Name=wheel_steering_fr_l, swing1LimitDegrees=45, swing2LimitDegrees=45, twistLimitDegrees=45}
 {bone1Name=shock_base_fr_l, bone2Name=steering_knuckle_fr_l, swing1LimitDegrees=45, swing2LimitDegrees=45, twistLimitDegrees=45}
 {bone1Name=steering_knuckle_fr_r, bone2Name=front_differential, swing1LimitDegrees=45, swing2LimitDegrees=45, twistLimitDegrees=45}
 {bone1Name=wheel_steering_fr_r, bone2Name=steering_knuckle_fr_r, swing1LimitDegrees=45, swing2LimitDegrees=45, twistLimitDegrees=45}
 {bone1Name=wheel_disc_fr_r, bone2Name=wheel_steering_fr_r, swing1LimitDegrees=45, swing2LimitDegrees=45, twistLimitDegrees=45}
 {bone1Name=shock_base_fr_r, bone2Name=steering_knuckle_fr_r, swing1LimitDegrees=45, swing2LimitDegrees=45, twistLimitDegrees=45}
 {bone1Name=axle_pivot_bk_l, bone2Name=frame, swing1LimitDegrees=45, swing2LimitDegrees=45, twistLimitDegrees=45}
 {bone1Name=wheel_disc_bk_l, bone2Name=axle_pivot_bk_l, swing1LimitDegrees=45, swing2LimitDegrees=45, twistLimitDegrees=45}
 {bone1Name=brakepad_bk_l, bone2Name=axle_pivot_bk_l, swing1LimitDegrees=45, swing2LimitDegrees=45, twistLimitDegrees=45}
 {bone1Name=shock_base_bk_l, bone2Name=axle_pivot_bk_l, swing1LimitDegrees=45, swing2LimitDegrees=45, twistLimitDegrees=45}
 {bone1Name=axle_pivot_bk_r, bone2Name=frame, swing1LimitDegrees=45, swing2LimitDegrees=45, twistLimitDegrees=45}
 {bone1Name=wheel_disc_bk_r, bone2Name=axle_pivot_bk_r, swing1LimitDegrees=45, swing2LimitDegrees=45, twistLimitDegrees=45}
 {bone1Name=shock_base_bk_r, bone2Name=axle_pivot_bk_r, swing1LimitDegrees=45, swing2LimitDegrees=45, twistLimitDegrees=45}
]
